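; Gameplay ability for the BushMud augment that, per its name, grants regeneration while the player is in stealth grass.
; Activation applies a configured gameplay effect, then waits for the InStealthGrass tag removal while a looping 28-second timer reapplies it.
; On removal, it applies the effect again if the avatar fails a configured tag check, then ends.
; The timer's condition reads that check's stored result rather than re-evaluating it.
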(event Custom|Reapply
  (if (not (stale Ability|EventActivateAbilityFromEvent _bdidactorhavetags))
    (Ability|ApplyGameplayEffectToOwner (Variables|Default|GetGameplay%20Effect%20Class))))

(event Ability|EventActivateAbilityFromEvent
  (params (EventData :type Gameplay%20Event%20Data%20Structure%20%28by%20ref%29))
  (bind _gameplay_effect_class (Variables|Default|GetGameplay%20Effect%20Class))
  ;[1 0.2718 0.06097] Activate when equipping bush
  (Ability|ApplyGameplayEffectToOwner _gameplay_effect_class)
  (Ability|Tasks|WaitGameplayTagRemove "(TagName=\"Gameplay.Action.Player.InStealthGrass\")"
    (:then
      (Utilities|Time|SetTimerbyEvent (delegate Custom|Reapply) 28.0 true))
    (:Removed
      (bind _bdidactorhavetags (Class|AthenaQuestFunctionLibrary|AthenaCheckActorHasTags (CallFunction|GetAvatarasFortPlayerPawn) (Variables|Default|GetTags%20to%20Check) true))
      (if (not _bdidactorhavetags)
        (Ability|ApplyGameplayEffectToOwner _gameplay_effect_class))
      (Ability|EndAbility))))
